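# Physics asset for the LastVoiceSteel two-handed pickaxe cosmetic, with capsule bodies (despite the Spheres name) on the handle and three dyn_ bones.
# Constraints chain the handle to dyn_LastVoiceSteel, dyn_Foot_0 and dyn_Foot_end, with the first two joints locked in swing and twist.
# Only the last joint can swing, with twist locked, so the foot end moves relative to the weapon.
# The two middle bodies are tiny link capsules between the larger handle and foot-end capsules.
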PhysicsAsset /BRCosmetics/Weapons/FORT_Melee/Pickaxe_LastVoiceSteel/Meshes/LastVoiceSteel_Axe_Physics_Spheres
bodies {
 Melee_TwoHanded_Handle [{
   shapeType=Capsule
   length=25.4327
   radius=6.2978
   shapeName=Melee_TwoHanded_Handle_capsule
   center=<0.0072 -1.0022 6.2583>
   rotation=<rot -3.6967 -90 0.0191>
  }]
 dyn_Foot_0 [{shapeType=Capsule, length=0.505, radius=0.505, shapeName=dyn_Foot_0_capsule, rotation=<rot 0 -90 0>}]
 dyn_Foot_end [{
   shapeType=Capsule
   length=16.2013
   radius=8.4603
   shapeName=dyn_Foot_end_capsule
   center=<0.5755 -1.3122 6.9667>
   rotation=<rot -0.8229 -90 4.4736>
  }]
 dyn_LastVoiceSteel [{shapeType=Capsule, length=0.505, radius=0.505, shapeName=dyn_LastVoiceSteel_capsule, rotation=<rot 0 -90 0>}]
}
constraints [
 {
  bone1Name=dyn_LastVoiceSteel
  bone2Name=Melee_TwoHanded_Handle
  swing1LimitDegrees=0
  swing1Motion=Locked
  swing2LimitDegrees=0
  swing2Motion=Locked
  twistLimitDegrees=0
  twistMotion=Locked
 }
 {bone1Name=dyn_Foot_end, bone2Name=dyn_Foot_0, swing1LimitDegrees=33, swing2LimitDegrees=20, twistLimitDegrees=3, twistMotion=Locked}
 {
  bone1Name=dyn_Foot_0
  bone2Name=dyn_LastVoiceSteel
  swing1LimitDegrees=0
  swing1Motion=Locked
  swing2LimitDegrees=0
  swing2Motion=Locked
  twistLimitDegrees=0
  twistMotion=Locked
 }
]
