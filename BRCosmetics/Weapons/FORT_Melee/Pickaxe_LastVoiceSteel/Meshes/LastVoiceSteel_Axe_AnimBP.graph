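# Animation Blueprint for the LastVoiceSteel pickaxe axe that adds secondary physics and a bobbing motion to the input pose.
# A front-end bool blends between a world-space RigidBody sim (Spheres asset) when true and a base-bone-space one (Sneaker asset) otherwise.
# A time-based sine then offsets dyn_Foot_end in Z.
# The update event tunes physics alpha and bob per context and damps acceleration/velocity scale when jumping outside the front end.
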AnimBlueprint /BRCosmetics/Weapons/FORT_Melee/Pickaxe_LastVoiceSteel/Meshes/LastVoiceSteel_Axe_AnimBP
target_skeleton=&/LastVoiceSteel_Axe_Skeleton
graphs [
 AnimGraph {nodes [
   n1 = Get scale() -> (scale:Vector)
   n2 = RigidBody!(
    ComponentPose=Local To Component(LocalPose=Use cached pose 'priseen'())
    Alpha=Get sneaker_FE_physics_alpha() -> (sneaker_FE_physics_alpha:double)
    ExternalForce
    ComponentLinearAccScale:Vector=@n1
    ComponentLinearVelScale:Vector=@n1
   ) -> (Pose) {
    bOverrideWorldGravity=true
    cachedBoundsScale=1
    overridePhysicsAsset=&/LastVoiceSteel_Axe_Physics_Sneaker
    simulationSpace=BaseBoneSpace
    alphaBoolBlend {blendOption=Linear}
    baseBoneRef {boneName=b_AR_Root}
    componentAppliedLinearAccClamp=<1000 1000 1000>
    componentLinearAccScale=<0.3 0.1 0.1>
    componentLinearVelScale=<0.3 0.1 0.1>
    overrideWorldGravity {z=980}
   }
   n3 = Get scale()
   n4 = RigidBody(
    ComponentPose=Local To Component(LocalPose=Use cached pose 'priseen'())
    Alpha=Get sneaker_FE_physics_alpha()
    ComponentLinearAccScale=@n3
    ComponentLinearVelScale=@n3
   ) {
    bOverrideWorldGravity=true
    cachedBoundsScale=1
    overridePhysicsAsset=&/LastVoiceSteel_Axe_Physics_Spheres
    simulationSpace=WorldSpace
    alphaBoolBlend {blendOption=Linear}
    baseBoneRef {boneName=dyn_LastVoiceSteel}
    componentAppliedLinearAccClamp=<1000 1000 1000>
    componentLinearAccScale=<0.3 0.1 0.1>
    componentLinearVelScale=<0.3 0.1 0.1>
    overrideWorldGravity {z=980}
   }
   n5 = Local To Component(LocalPose=Blend Poses by bool(BlendPose_0=Component To Local(ComponentPose=@n4), BlendPose_1=Component To Local(ComponentPose=@n2), bActiveValue=Get bIsInFrontEnd()) {blendType=Linear})
   n6 = Get foot() -> (foot:double)
   Output Pose(Result=Component To Local(ComponentPose="Transform (Modify) Bone - Bone: dyn_Foot_end"!(ComponentPose=@n5, Alpha=1, Translation_X:double=0, Translation_Y:double=0, Translation_Z:double=@n6, Rotation="0,0,0", Scale="1,1,1") -> (Pose) {translationMode=BMM_Additive, alphaBoolBlend {blendOption=Linear}, boneToModify {boneName=dyn_Foot_end}})) {layerGroup=None}
   AnimGraphNode_SubInput_0 = Input Pose()
   Save cached pose 'priseen'(Pose=@AnimGraphNode_SubInput_0) {cacheName=priseen}
  ]}
 EventGraph {nodes [
   Event BlueprintUpdateAnimation()
   > n1 = Branch(Condition=Get bIsInFrontEnd())
   > Set sneaker_FE_float_distance_alpha(sneaker_FE_float_distance_alpha:double=1.5) -> (Output_Get:double)
   > Set sneaker_FE_physics_alpha(sneaker_FE_physics_alpha:double=0.2) -> (Output_Get:double)
   n2 = float + float(A=float * float(A=GetRealTimeSeconds(), B=1.4) {is=title K2Node_PromotableOperator}, B=65)
   > Set foot(foot:double=float * float(A="Sin (Radians)"(A=@n2), B=Get sneaker_FE_float_distance_alpha() -> (sneaker_FE_float_distance_alpha:double))) -> (Output_Get:double)
   > Set scale(scale:Vector="0.3,0.1,0.1") -> (Output_Get:Vector)
   > Set scale_spher(scale_spher:Vector="1.5,1.5,1.5") -> (Output_Get:Vector)
   Set sneaker_FE_float_distance_alpha(execute=@n1.else, sneaker_FE_float_distance_alpha=2)
   > Set sneaker_FE_physics_alpha(sneaker_FE_physics_alpha=0.5)
   n3 = float + float(A=float * float(A=GetRealTimeSeconds(), B=1.5), B=77)
   > Set foot(foot=float * float(A="Sin (Radians)"(A=@n3), B=Get sneaker_FE_float_distance_alpha()))
   > n4 = Branch(Condition=Get bIsJumping())
   > Set scale(scale="0.05,0.05,0.05")
   > Set scale_spher(scale_spher="0.5,0.5,0.5")
   Set scale(execute=@n4.else, scale="0.3,0.1,0.1")
   > Set scale_spher(scale_spher="1.5,1.5,1.5")
   TryGetPawnOwner()
  ]}
]
variables [
 boolean {type=boolean}
 foot {type=number}
 foot_z {type=number}
 parametr {type=number}
 scale {type=Vector}
 scale_spher {type=Vector}
 sneaker_FE_float_distance_alpha {type=number}
 sneaker_FE_physics_alpha {type=number}
 sphere_sin_0 {type=number}
 sphere_sin_1 {type=number}
 sphere_sin_2 {type=number}
 sphere_sin_3 {type=number}
 sphere_sin_4 {type=number}
 sphere_sin_5 {type=number}
 sphere_sin_6 {type=number}
 sphere_sin_7 {type=number}
 sphere_sin_8 {type=number}
 sphere_sin_9 {type=number}
 Translation {type=Vector}
]
